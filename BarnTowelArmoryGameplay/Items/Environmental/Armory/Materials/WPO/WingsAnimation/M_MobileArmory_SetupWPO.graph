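# Mobile Armory setup material that animates four wings through World Position Offset, with simple parameter-driven PBR inputs.
# Local-position masks derive each wing's axis, pivot and delay, and vertex color red weights a RotateVertexAboutAxis call that also supplies the corrected normal.
# Sine flapping plus a base angle follows eased landing and closing timings measured from BP_GetTime.
# A chained pod-rotation call exists but does not feed the switchable WPO output.
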Material /BarnTowelArmoryGameplay/Items/Environmental/Armory/Materials/WPO/WingsAnimation/M_MobileArmory_SetupWPO
graph {
 properties {
  bAutomaticallySetUsageInEditor=false
  bUseMaterialAttributes=true
  bUsedWithClothing=true
  bUsedWithNanite=true
  maxWorldPositionOffsetDisplacement=5000
  usedShadingModels=Default Lit
 }
 nodes [
  n1 = ScalarParameter() {defaultValue=1, group=WPO_Wings, parameterName=Wings_InitialFlappingSpeed}
  n2 = Add(A=Time(), B=NamedRerouteUsage())
  n3 = Multiply(A=@n2, B=@n1)
  n4 = NamedRerouteUsage()
  n5 = Saturate(None=@n4)
  LinearInterpolate(A=Multiply(A=@n1, B=@n2), B=@n3, Alpha=@n5)
  n6 = LinearInterpolate(
   A=ScalarParameter() {defaultValue=1, group=WPO_Wings, parameterName=Wings_InitialFlappingIntensity}
   B=ScalarParameter() {defaultValue=1, group=WPO_Wings, parameterName=Wings_LandingFlappingIntensity}
   Alpha=@n5
  )
  LinearInterpolate(A=@n6, Alpha=NamedRerouteUsage()) {constB=0}
  MakeMaterialAttributes!(
   BaseColor=VectorParameter() {group=01_PBR, parameterName=BaseColor, sortPriority=0, defaultValue=<rgba 1 0.0256 0 1>}.RGB
   Metallic=ScalarParameter() {group=01_PBR, parameterName=Metallic, sortPriority=1}
   Specular=ScalarParameter() {defaultValue=0.5, group=01_PBR, parameterName=Specular, sortPriority=2}
   Roughness=ScalarParameter() {defaultValue=0.5, group=01_PBR, parameterName=Roughness, sortPriority=2}
   Anisotropy
   EmissiveColor
   Opacity
   OpacityMask
   Normal=NamedRerouteUsage()
   Tangent
   WorldPositionOffset=NamedRerouteUsage()
   SubsurfaceColor
   ClearCoat
   ClearCoatRoughness
   AmbientOcclusion
   Refraction
   CustomizedUVs_0
   CustomizedUVs_1
   CustomizedUVs_2
   CustomizedUVs_3
   CustomizedUVs_4
   CustomizedUVs_5
   CustomizedUVs_6
   CustomizedUVs_7
   PixelDepthOffset
   ShadingModel
   Displacement
  )
  n7 = Constant3Vector() {constant=<rgba 0 0 1 0>}
  n8 = Sine(None=@n3)
  n9 = Add(
   A=Multiply(A=Multiply(A=@n8) {constB=0.01}, B=@n6)
   B=LinearInterpolate(
    A=ScalarParameter() {defaultValue=0.4, group=WPO_Wings, parameterName=Wings_InitialRotationAngle, sortPriority=0}
    B=ScalarParameter() {defaultValue=0.2, group=WPO_Wings, parameterName=Wings_LandinglRotationAngle, sortPriority=0}
    Alpha=NamedRerouteUsage()
   ) {constB=0}
  )
  n10 = NamedRerouteUsage()
  n11 = VertexColor()
  n12 = LinearInterpolate(A=Constant3Vector() {constant=<rgba 0 0 1 0>}, B=NamedRerouteUsage(), Alpha=@n11.R)
  n13 = MaterialFunctionCall!(
   RotationAngle=Multiply(A=LinearInterpolate(A=@n9, Alpha=@n10) {constB=0}, B=@n11.R)
   NormalizedRotationAxis=Transform(None=Normalize(VectorInput=@n12)) {transformSourceType=TRANSFORMSOURCE_Local}
   PivotPoint=TransformPosition(None=NamedRerouteUsage()) {transformType=TRANSFORMPOSSOURCE_World}
   Position
   Normal
   TangentU
   TangentV
  ) {materialFunction=/Game/Characters/Player/Male/Medium/Bodies/M_MED_Inferno/FX/Mat/RotateVertexAboutAxis}
  MaterialFunctionCall(
   RotationAngle=LinearInterpolate(A=ScalarParameter() {parameterName=InitialPodRotationValue}, Alpha=NamedRerouteUsage()) {constB=0}
   NormalizedRotationAxis=Normalize(VectorInput=Transform(None=@n7) {transformSourceType=TRANSFORMSOURCE_Local})
   PivotPoint=TransformPosition(None=Constant3Vector()) {transformType=TRANSFORMPOSSOURCE_World}
   Position=Add(A=WorldPosition().XYZ, B=@n13.WPO)
   Normal=@n13.Normal
  ) {materialFunction=&RotateVertexAboutAxis}
  Multiply(A=@n4) {constB=2.5}
  n14 = ScalarParameter() {defaultValue=9999999827968, group=WPO_Wings, parameterName=BP_GetTime, sortPriority=0}
  n15 = Time()
  n16 = Subtract(A=@n15, B=@n14)
  n17 = ScalarParameter() {defaultValue=1, group=WPO_Rotation, parameterName=FallingRotationDuration}
  n18 = OneMinus(None=Saturate(None=Divide(A=@n16, B=@n17)))
  n19 = Multiply(A=@n18, B=@n18)
  Multiply(A=@n19, B=@n19)
  n20 = LocalPosition()
  n21 = ComponentMask(None=@n20.XYZ) {g=true}
  n22 = Add(A=@n21) {constB=-450}
  n23 = OneMinus(None=Saturate(None=Add(A=@n21) {constB=450}))
  n24 = NamedRerouteDeclaration(None=Add(A=Saturate(None=@n22), B=Multiply(A=@n23) {constB=-1})) {name=WingsMask_Y, nodeColor=<rgba 1 0 0.7969 1>}
  n25 = Constant3Vector() {constant=<rgba 1 0 0 0>}
  n26 = ComponentMask(None=@n20.XYZ) {r=true}
  n27 = Saturate(None=Add(A=@n26) {constB=-450})
  n28 = OneMinus(None=Saturate(None=Add(A=@n26) {constB=450}))
  n29 = NamedRerouteDeclaration(None=Add(A=Multiply(A=@n27) {constB=-1}, B=@n28)) {name=WingMask_X, nodeColor=<rgba 0 1 0.8828 1>}
  n30 = Constant3Vector() {constant=<rgba 0 1 0 0>}
  NamedRerouteDeclaration(None=Add(A=Multiply(A=@n24, B=@n25), B=Multiply(A=@n29, B=@n30))) {name=RotationAxis, nodeColor=<rgba 0.8047 0 1 1>}
  NamedRerouteDeclaration(None=OneMinus(None=@n19)) {name=PodRotation_Timing, nodeColor=<rgba 1 0 0.2344 1>}
  n31 = ScalarParameter() {defaultValue=545, group=WPO_Wings, parameterName=Pivot_Depth}
  n32 = Multiply(A=@n31) {constB=-1}
  n33 = NamedRerouteUsage()
  n34 = NamedRerouteUsage()
  NamedRerouteDeclaration(None=MaterialFunctionCall(
    R=Multiply(A=@n32, B=@n33) {constB=560}
    G=Multiply(A=@n34, B=@n31) {constB=560}
    B=ScalarParameter() {defaultValue=2190, group=WPO_Wings, parameterName=WingsPivot_Height}
   ) {materialFunction=&AppendMany}.RGB) {name=Wings_Pivot, nodeColor=<rgba 1 0.5625 0 1>}
  n35 = ComponentMask(None=@n20.XYZ) {b=true}
  NamedRerouteDeclaration(None=Saturate(None=Add(A=@n35) {constB=-800})) {name=WingsHeightMask, nodeColor=<rgba 0.5703 1 0 1>}
  NamedRerouteDeclaration(None=StaticSwitchParameter(True=@n13.WPO, False=Constant3Vector()) {defaultValue=true, parameterName=UseWPOWings}) {name=WPO, nodeColor=<rgba 1 0 0.8906 1>}
  n36 = ScalarParameter() {defaultValue=0.2, group=WPO_Wings, parameterName=Wing_03_Delay, sortPriority=10}
  n37 = Multiply(A=ScalarParameter() {defaultValue=0.1, group=WPO_Wings, parameterName=Wing_02_Delay, sortPriority=10}, B=@n27)
  n38 = Multiply(A=@n28, B=ScalarParameter() {defaultValue=0.3, group=WPO_Wings, parameterName=Wing_04_Delay, sortPriority=10})
  NamedRerouteDeclaration(None=Add(A=Multiply(A=@n36, B=@n23), B=Add(A=@n37, B=@n38))) {name=WingsOffset, nodeColor=<rgba 0 0.0934 1 1>}
  NamedRerouteDeclaration(None=Transform(None=@n13.Normal) {transformSourceType=TRANSFORMSOURCE_World, transformType=TRANSFORM_Tangent}) {name=OutputNormal, nodeColor=<rgba 1 0.8438 0 1>}
  n39 = ScalarParameter() {defaultValue=1, group=WPO_Wings, parameterName=Wings_LandingAnimationDelay}
  n40 = Subtract(A=Subtract(A=@n15, B=@n39), B=@n14)
  n41 = ScalarParameter() {defaultValue=1, group=WPO_Wings, parameterName=Wings_LandingAnimationSpeed}
  n42 = OneMinus(None=Saturate(None=Multiply(A=@n40, B=@n41)))
  n43 = Multiply(A=@n42, B=@n42)
  n44 = OneMinus(None=Multiply(A=@n43, B=@n43))
  NamedRerouteDeclaration(None=@n44) {name=WingsLanding_Timing, nodeColor=<rgba 0 1 0.5078 1>}
  n45 = ScalarParameter() {defaultValue=1, group=WPO_Wings, parameterName=Wings_ClosingAnimationSpeed}
  n46 = ScalarParameter() {defaultValue=1, group=WPO_Wings, parameterName=Wings_ClosingAnimationDelay}
  n47 = Subtract(A=Subtract(A=@n15, B=Add(A=@n46, B=@n39)), B=@n14)
  n48 = SmoothStep(Value=Saturate(None=Multiply(A=@n45, B=@n47)))
  NamedRerouteDeclaration(None=@n48) {name=WingsClosing_Timing, nodeColor=<rgba 1 0.1875 0 1>}
  Reroute(None=@n44)
  Reroute(None=@n48)
  ScalarParameter() {defaultValue=1, group=WPO_Wings, parameterName=Wings_LandingFlappingSpeed}
  Transform(None=Constant3Vector() {constant=<rgba 0 0 1 0>}) {transformSourceType=TRANSFORMSOURCE_Local}
 ]
}
