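# Fortnite layered base material for the Mobile Armory that animates its wings through World Position Offset as it lands and closes.
# Shared damage, wind, snow, veneer, biome and RVT functions build the surface, with a low-quality mobile simplification.
# Local-position masks split four wings, giving each a rotation axis, pivot and staggered sine flap.
# Time since BP_GetTime drives landing and closing curves; vertex red weights RotateVertexAboutAxis, which produces offset and corrected normal.
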Material /BarnTowelArmoryGameplay/Items/Environmental/Armory/Materials/WPO/WingsAnimation/M_MobileArmory_WingsWPO_PBR
graph {
 properties {
  bAutomaticallySetUsageInEditor=false
  bTangentSpaceNormal=false
  bUseMaterialAttributes=true
  bUsedWithInstancedStaticMeshes=true
  bUsedWithNanite=true
  bUsedWithSplineMeshes=true
  ditheredLODTransition=true
  pixelDepthOffsetMode=PDOM_Legacy
  shadingModel=MSM_FromMaterialExpression
  usedShadingModels=Default Lit | Subsurface | Clear Coat | Two Sided Foliage
 }
 nodes [
  n1 = MaterialFunctionCall() {materialFunction=/Game/Global/Base4Layers_2022/MaterialFunctions/MF_GigaMix_2023}
  MaterialExpressionReroute_70 = Reroute(None=@n1.Result)
  n2 = ScalarParameter() {
   desc=Documentation link
   group="0 (Expand)  Documentation (RMB - Copy display name) and Debug view"
   parameterName=https://confluence.it.epicgames.com/display/FORT/Fortnite+Base+Material+2022
  }
  n3 = SetMaterialAttributes!(MaterialAttributes=@MaterialExpressionReroute_70.MaterialAttributes, World Position Offset=Multiply(A=@n2)) {attributeSetTypes [#1]}
  MaterialExpressionReroute_71 = Reroute(None=@n3)
  MaterialExpressionReroute_69 = Reroute(None=@MaterialExpressionReroute_71.Output)
  n4 = MaterialFunctionCall!(Material Attributes=MaterialFunctionCall!(In=MaterialFunctionCall(Material_In=@MaterialExpressionReroute_69.MaterialAttributes) {materialFunction=&MF_FN_DestructibleDamage}.Material_Out) {materialFunction=/Game/Athena/WinterArt/Materials/MF_WPO_WindNew}.Result) {materialFunction=/Game/Creative/Sets/MilitaryBase/Props/MF_FanSpin}
  n5 = MaterialFunctionCall!(In=MaterialFunctionCall!(
    In=MaterialFunctionCall!(In=@n4.Result, UseSphericalMappingOverride) {materialFunction=/Game/Packages/Fortress_SharedMaterials/Base_Material_Functions/MF_CurieFallbackEffect}.Result
    Custom Mask
    Custom Rotation Axis
   ) {materialFunction=/Game/Athena/WinterArt/Materials/MF_WindWPO_v2}.Result) {materialFunction=/Game/Global/Base4Layers_2022/MaterialFunctions/MF_SnowAndIceArchitecture}
  MaterialExpressionReroute_68 = Reroute(None=@n5.Result)
  Composite!(None=@MaterialExpressionReroute_68.Result) {desc=Transform to World in case we will use RVT after this point, subgraphName=Tangent to World}
  Composite!(MaterialAttributes=@n1.Result) {subgraphName=Documentation Link}
  Composite(MaterialAttributes=@MaterialExpressionReroute_71.Output) {desc=Game Shared WPO and Fire Effect, subgraphName=WPO And Shared Functions}
  MaterialExpressionReroute_1 = Reroute(None=@n12.RVT H)
  MaterialExpressionReroute_3 = Reroute(None=@MaterialExpressionReroute_68.Result)
  n6 = GetMaterialAttributes(None=@MaterialExpressionReroute_3.Input) {attributeGetTypes [#2]}
  n7 = Transform(None=@n6.Normal)
  n8 = LinearInterpolate(A=Multiply(A=@n7) {constB=-1}, B=@n7, Alpha=Saturate(None=TwoSidedSign())) {desc="Fixes normals\r\non backfaces"}
  n9 = SetMaterialAttributes!(
   MaterialAttributes=@n6.MaterialAttributes
   Normal=StaticSwitchParameter(True=MaterialProxyReplace(Realtime=@n8, MaterialProxy=@n6.Normal), False=@n8) {parameterName=Force TangentSpaceNormalsFor HLOD ProxyReplace}
  ) {attributeSetTypes [#2]}
  MaterialExpressionReroute_2 = Reroute(None=@n9)
  n10 = MaterialFunctionCall!(In=MaterialFunctionCall!(In=@MaterialExpressionReroute_2.Output) {materialFunction=/Game/ArtTools/Tech_Art/DirectFX/Assets/MaterialFunctions/MF_DirectFX}.Result) {materialFunction=/Game/ArtTools/Tech_Art/Veneer/MaterialFunctions/MF_Veneer}
  MaterialExpressionReroute_8 = Reroute(None=@n10.Result)
  n11 = MaterialFunctionCall!(AO, RVT_Height=@MaterialExpressionReroute_1."RVT_Height (S)", For Buildings=@MaterialExpressionReroute_8."For Moss (MA)", For Moss) {materialFunction=/Game/Environments/Asteria/Rocks/Materials/MaterialFunctions/MF_JungleBiome}
  MaterialExpressionReroute_0 = Reroute(None=@n11.Buildings Jungl)
  n12 = MaterialFunctionCall!(
   In=@MaterialExpressionReroute_0.Result
   "RVT Height Map (Roughness)"=GetMaterialAttributes(None=@n10.Result) {attributeGetTypes [#3]}.Roughness
  ) {materialFunction=/Game/Global/Base4Layers_2022/MaterialFunctions/MF_RVTBlending}
  Composite!(InputPin=@n12.Result) {subgraphName=Mobile Opt}
  Composite!("For Moss (MA)"=@n10.Result, "RVT_Height (S)"=@n12.RVT H) {subgraphName="Seasonal "}
  n13 = ScalarParameter() {defaultValue=1, group=WPO_Wings, parameterName=Wings_InitialFlappingSpeed}
  n14 = Add(A=Time(), B=NamedRerouteUsage())
  n15 = Multiply(A=@n14, B=@n13)
  n16 = NamedRerouteUsage()
  n17 = Saturate(None=@n16)
  LinearInterpolate(A=Multiply(A=@n13, B=@n14), B=@n15, Alpha=@n17)
  n18 = LinearInterpolate(
   A=ScalarParameter() {defaultValue=1, group=WPO_Wings, parameterName=Wings_InitialFlappingIntensity}
   B=ScalarParameter() {defaultValue=1, group=WPO_Wings, parameterName=Wings_LandingFlappingIntensity}
   Alpha=@n17
  )
  LinearInterpolate(A=@n18, Alpha=NamedRerouteUsage()) {constB=0}
  Multiply(A=@n16) {constB=2.5}
  n19 = ScalarParameter() {defaultValue=9999999827968, group=WPO_Wings, parameterName=BP_GetTime, sortPriority=0}
  n20 = Time()
  n21 = Subtract(A=@n20, B=@n19)
  n22 = ScalarParameter() {defaultValue=1, group=WPO_Rotation, parameterName=FallingRotationDuration}
  n23 = OneMinus(None=Saturate(None=Divide(A=@n21, B=@n22)))
  n24 = Multiply(A=@n23, B=@n23)
  Multiply(A=@n24, B=@n24)
  n25 = LocalPosition()
  n26 = ComponentMask(None=@n25.XYZ) {g=true}
  n27 = Add(A=@n26) {constB=-450}
  n28 = OneMinus(None=Saturate(None=Add(A=@n26) {constB=450}))
  n29 = NamedRerouteDeclaration(None=Add(A=Saturate(None=@n27), B=Multiply(A=@n28) {constB=-1})) {name=WingsMask_Y, nodeColor=<rgba 1 0 0.7969 1>}
  n30 = Constant3Vector() {constant=<rgba 1 0 0 0>}
  n31 = ComponentMask(None=@n25.XYZ) {r=true}
  n32 = Saturate(None=Add(A=@n31) {constB=-450})
  n33 = OneMinus(None=Saturate(None=Add(A=@n31) {constB=450}))
  n34 = NamedRerouteDeclaration(None=Add(A=Multiply(A=@n32) {constB=-1}, B=@n33)) {name=WingMask_X, nodeColor=<rgba 0 1 0.8828 1>}
  n35 = Constant3Vector() {constant=<rgba 0 1 0 0>}
  NamedRerouteDeclaration(None=Add(A=Multiply(A=@n29, B=@n30), B=Multiply(A=@n34, B=@n35))) {name=RotationAxis, nodeColor=<rgba 0.8047 0 1 1>}
  NamedRerouteDeclaration(None=OneMinus(None=@n24)) {name=PodRotation_Timing, nodeColor=<rgba 1 0 0.2344 1>}
  MaterialExpressionReroute_7 = Reroute(None=@n12.Result)
  n36 = GetMaterialAttributes(None=@MaterialExpressionReroute_7.InputPin)
  n37 = Constant() {r=0.5}
  n38 = Constant() {r=0.8}
  n39 = ShadingPathSwitch(
   Default=@MaterialExpressionReroute_7.InputPin
   Mobile=QualitySwitch(
    Default=@MaterialExpressionReroute_7.InputPin
    Low=SetMaterialAttributes!(MaterialAttributes=@n36.MaterialAttributes, Specular=@n37, Roughness=@n38) {attributeSetTypes [#4, #3]}
    High=@MaterialExpressionReroute_7.InputPin
    Medium=@MaterialExpressionReroute_7.InputPin
    Epic=@MaterialExpressionReroute_7.InputPin
   )
  )
  MaterialExpressionReroute_6 = Reroute(None=@n39)
  n40 = MaterialFunctionCall(MA In=@MaterialExpressionReroute_6.Output) {materialFunction=&MF_CreativeMoveTool}
  NamedRerouteDeclaration(None=GetMaterialAttributes(None=@n40.MA Out) {attributeGetTypes [#2]}.Normal) {name=InputNormal, nodeColor=<rgba 0.5703 1 0 1>}
  n41 = ScalarParameter() {defaultValue=545, group=WPO_Wings, parameterName=Pivot_Depth}
  n42 = Multiply(A=@n41) {constB=-1}
  n43 = NamedRerouteUsage()
  n44 = NamedRerouteUsage()
  NamedRerouteDeclaration(None=MaterialFunctionCall(
    R=Multiply(A=@n42, B=@n43) {constB=560}
    G=Multiply(A=@n44, B=@n41) {constB=560}
    B=ScalarParameter() {defaultValue=2190, group=WPO_Wings, parameterName=WingsPivot_Height}
   ) {materialFunction=&AppendMany}.RGB) {name=Wings_Pivot, nodeColor=<rgba 1 0.5625 0 1>}
  n45 = ComponentMask(None=@n25.XYZ) {b=true}
  NamedRerouteDeclaration(None=Saturate(None=Add(A=@n45) {constB=-800})) {name=WingsHeightMask, nodeColor=<rgba 0.5703 1 0 1>}
  n46 = Sine(None=@n15)
  n47 = Add(
   A=Multiply(A=Multiply(A=@n46) {constB=0.01}, B=@n18)
   B=LinearInterpolate(
    A=ScalarParameter() {defaultValue=0.4, group=WPO_Wings, parameterName=Wings_InitialRotationAngle, sortPriority=0}
    B=ScalarParameter() {defaultValue=0.2, group=WPO_Wings, parameterName=Wings_LandinglRotationAngle, sortPriority=0}
    Alpha=NamedRerouteUsage()
   ) {constB=0}
  )
  n48 = NamedRerouteUsage()
  n49 = VertexColor()
  n50 = LinearInterpolate(A=Constant3Vector() {constant=<rgba 0 0 1 0>}, B=NamedRerouteUsage(), Alpha=@n49.R)
  n51 = TextureSampleParameter2D(UVs=Multiply(A=TextureCoordinate(), B=ScalarParameter() {defaultValue=1, group=1_Layer UV, parameterName=UV Tiling, sortPriority=1})) {
   group=01_PBR
   parameterName=NormalTexture
   samplerType=SAMPLERTYPE_Normal
   texture=/BarnTowelArmoryGameplay/Items/Environmental/Armory/Textures/T_MobileArmory_PaintedMetal_N
  }
  n52 = MaterialFunctionCall!(
   RotationAngle=Multiply(A=LinearInterpolate(A=@n47, Alpha=@n48) {constB=0}, B=@n49.R)
   NormalizedRotationAxis=Transform(None=Normalize(VectorInput=@n50)) {transformSourceType=TRANSFORMSOURCE_Local}
   PivotPoint=TransformPosition(None=NamedRerouteUsage()) {transformType=TRANSFORMPOSSOURCE_World}
   Position
   Normal=Transform(None=@n51.RGB)
   TangentU
   TangentV
  ) {materialFunction=/Game/Characters/Player/Male/Medium/Bodies/M_MED_Inferno/FX/Mat/RotateVertexAboutAxis}
  NamedRerouteDeclaration(None=StaticSwitchParameter(True=@n52.WPO, False=Constant3Vector()) {defaultValue=true, parameterName=UseWPOWings}) {name=WPO, nodeColor=<rgba 1 0 0.8906 1>}
  n53 = ScalarParameter() {defaultValue=0.2, group=WPO_Wings, parameterName=Wing_03_Delay, sortPriority=10}
  n54 = Multiply(A=ScalarParameter() {defaultValue=0.1, group=WPO_Wings, parameterName=Wing_02_Delay, sortPriority=10}, B=@n32)
  n55 = Multiply(A=@n33, B=ScalarParameter() {defaultValue=0.3, group=WPO_Wings, parameterName=Wing_04_Delay, sortPriority=10})
  NamedRerouteDeclaration(None=Add(A=Multiply(A=@n53, B=@n28), B=Add(A=@n54, B=@n55))) {name=WingsOffset, nodeColor=<rgba 0 0.0934 1 1>}
  NamedRerouteDeclaration(None=@n52.Normal) {name=OutputNormal, nodeColor=<rgba 1 0.8438 0 1>}
  n56 = ScalarParameter() {defaultValue=1, group=WPO_Wings, parameterName=Wings_LandingAnimationDelay}
  n57 = Subtract(A=Subtract(A=@n20, B=@n56), B=@n19)
  n58 = ScalarParameter() {defaultValue=1, group=WPO_Wings, parameterName=Wings_LandingAnimationSpeed}
  n59 = OneMinus(None=Saturate(None=Multiply(A=@n57, B=@n58)))
  n60 = Multiply(A=@n59, B=@n59)
  n61 = OneMinus(None=Multiply(A=@n60, B=@n60))
  NamedRerouteDeclaration(None=@n61) {name=WingsLanding_Timing, nodeColor=<rgba 0 1 0.5078 1>}
  n62 = ScalarParameter() {defaultValue=1, group=WPO_Wings, parameterName=Wings_ClosingAnimationSpeed}
  n63 = ScalarParameter() {defaultValue=1, group=WPO_Wings, parameterName=Wings_ClosingAnimationDelay}
  n64 = Subtract(A=Subtract(A=@n20, B=Add(A=@n63, B=@n56)), B=@n19)
  n65 = SmoothStep(Value=Saturate(None=Multiply(A=@n62, B=@n64)))
  NamedRerouteDeclaration(None=@n65) {name=WingsClosing_Timing, nodeColor=<rgba 1 0.1875 0 1>}
  PinBase() {reroutePins [Input {expression=&self:MaterialExpressionReroute_3}]}
  PinBase!(Output=@n9) {reroutePins [Output {expression=&self:MaterialExpressionReroute_2}]}
  PinBase!() {reroutePins [
    "For Moss (MA)" {expression=&self:MaterialExpressionReroute_8}
    "RVT_Height (S)" {expression=&self:MaterialExpressionReroute_1}
   ]}
  PinBase!(Result=@n11.Buildings Jungl) {reroutePins [Result {expression=&self:MaterialExpressionReroute_0}]}
  PinBase!() {reroutePins [MaterialAttributes {expression=&self:MaterialExpressionReroute_70}]}
  PinBase!(Output=@n3) {reroutePins [Output {expression=&self:MaterialExpressionReroute_71}]}
  PinBase!() {reroutePins [MaterialAttributes {expression=&self:MaterialExpressionReroute_69}]}
  PinBase!(Result=@n5.Result) {reroutePins [Result {expression=&self:MaterialExpressionReroute_68}]}
  PinBase!() {reroutePins [InputPin {expression=&self:MaterialExpressionReroute_7}]}
  PinBase!(Output=@n39) {reroutePins [Output {expression=&self:MaterialExpressionReroute_6}]}
  Reroute(None=@n61)
  Reroute(None=@n65)
  ScalarParameter() {defaultValue=1, group=WPO_Wings, parameterName=Wings_LandingFlappingSpeed}
  SetMaterialAttributes!(MaterialAttributes=@n40.MA Out, World Position Offset=NamedRerouteUsage(), Normal=NamedRerouteUsage()) {attributeSetTypes [#1, #2]}
 ]
}
